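# Physics asset for the OatmealSpread gadget prop mesh, giving it simplified collision and joint shapes on three bones.
# Each of root, offset and main has one capsule so short it is nearly a sphere: large on root and main, tiny on offset.
# Two constraints chain root to offset to main, each with matching 45-degree swing and twist limit angles.
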PhysicsAsset /BRCosmetics/Animation/Props/Gadgets_OatmealSpread/Meshes/PA_OatmealSpread_Gadget
bodies {
 main [{shapeType=Capsule, length=0.505, radius=104.0003, rotation=<rot 0 -90 0>}]
 offset [{shapeType=Capsule, length=0.505, radius=0.505, rotation=<rot 0 -90 0>}]
 root [{shapeType=Capsule, length=5.8315, radius=189.3792, rotation=<rot 0 -90 0>}]
}
constraints [
 {bone1Name=offset, bone2Name=root, swing1LimitDegrees=45, swing2LimitDegrees=45, twistLimitDegrees=45}
 {bone1Name=main, bone2Name=offset, swing1LimitDegrees=45, swing2LimitDegrees=45, twistLimitDegrees=45}
]
